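# Niagara idle eye glow for the ClawedRaven character cosmetic, drawn as two bright magenta sprites nudged toward the camera.
# A single burst spawns both particles, each partitioned by execution index to select its own bone or socket.
# SkeletalMeshLocation re-samples the system-level skeletal mesh every update so the glows track the character.
# Particles persist without lifetime kill, and a front-end modifier lowers their alpha outside gameplay.
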NiagaraSystem /BRCosmetics/Characters/Player/Female/Medium/Bodies/F_MED_ClawedRaven/FX/NS_ClawedRaven_Character_EyeGlow_Idle
dependencies {
 usedDataInterfaces [&NiagaraDataInterfaceSkeletalMesh]
 usedDynamicInputs []
 usedModules [
  &self:SystemSpawnScript.SystemScriptSource.SystemScriptGraph.NiagaraNodeAssignment_1.SetVariables_#1
  &SystemState
 ]
}
emitters {EyeGlows {
  emitterUpdateScript [
   &EmitterState(
    Inactive Response=::NewEnumerator1
    Life Cycle Mode=::NewEnumerator1
    Loop Behavior=::NewEnumerator1
    Loop Duration Mode=::NewEnumerator1
    UseLoopDelay=false
   )
   &SpawnBurst_Instantaneous(Spawn Count=2, Use Spawn Probability=false, UseLoopCountLimit=true)
  ]
  particleSpawnScript [
   &InitializeParticle(
    Color=<rgba 4.2333 0.4 5 0.4>
    Sprite Size Mode=::NewEnumerator3
    Uniform Sprite Size=11
    UsePositionOffset=false
    Write Material Random=true
   )
   &PartitionParticles(ID to Partition:NiagaraInt32=&ReturnExecIndex())
   &CameraOffset(Camera Offset Amount=5)
  ]
  particleUpdateScript [
   /SetVariables(Particles.Position=&AddVectorToPosition(Position=<0 0 0>, Vector=<0 0 0>)) {isSetParameters=true}
   &ParticleState(Kill Particles When Lifetime Has Elapsed=false, Let Infinitely Lived Particles Die When Emitter Deactivates=true)
   &SkeletalMeshLocation(
    Bone / Socket Index=$Particles.PartitionParticles.Partition
    MaskSpawnGroup=false
    Mesh Sampling Type=::NewEnumerator5
    Sampled Position Offset=<0 -0.2 0>
    Skeletal Mesh=$System.Skeletal Mesh
    Socket Sampling Mode=::NewEnumerator5
   )
   &ScaleColor(Scale Alpha=&FN_FrontEndFloatModifier(Gameplay=1, Front End=0.35), ScaleA=true, ScaleRGB=false)
  ]
  rendererClasses [&NiagaraSpriteRendererProperties]
  renderers [{rendererClass=&NiagaraSpriteRendererProperties}]
 }}
summary {
 systemName=NS_ClawedRaven_Character_EyeGlow_Idle
 emitters [{emitterName=EyeGlows, rendererClasses [&NiagaraSpriteRendererProperties]}]
 userVariables [
  User.bIsFrontEnd {type=NiagaraBool}
  User.bIsFrontEndPreview {type=NiagaraBool}
  User.bKillReaction {type=NiagaraBool}
  User.global Alpha {type=NiagaraFloat, defaultValue=0, description=""}
  User.IsSkydiving {type=NiagaraBool}
  User.showFX {type=NiagaraFloat, defaultValue=1, description=""}
 ]
}
