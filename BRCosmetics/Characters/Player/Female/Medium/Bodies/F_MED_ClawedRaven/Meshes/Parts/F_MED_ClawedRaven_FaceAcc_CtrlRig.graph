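# Control Rig for the ClawedRaven face accessory that applies pose-driven corrective offsets to the head feather bones as the upper arms move.
# A spherical pose reader on upperarm_r weights offsets on three feather corrective bones.
# Two upperarm_l readers target the same other two bones: IsParachuteOpened is subtracted from one and multiplies the other, switching between normal and parachute correctives.
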ControlRigBlueprint /BRCosmetics/Characters/Player/Female/Medium/Bodies/F_MED_ClawedRaven/Meshes/Parts/F_MED_ClawedRaven_FaceAcc_CtrlRig
graphs [
 RigVMFunctionLibrary {links [], nodes []}
 RigVMModel {nodes [
   RigUnit_BeginExecution = &RigUnit_BeginExecution()
   n1 = &RigUnit_SphericalPoseReader(
    ActiveRegionScaleFactors=(PositiveWidth=0.7,NegativeWidth=1.5,PositiveHeight=0.7,NegativeHeight=1.5)
    ActiveRegionSize=0.1
    Debug=(bDrawDebug=false,bDraw2D=False,bDrawLocalAxes=False,DebugScale=25,DebugSegments=20,DebugThickness=0.25)
    DriverAxis=(X=-1,Y=0,Z=0)
    DriverItem=(Type=Bone,Name="upperarm_r")
    FalloffRegionScaleFactors=(PositiveWidth=1,NegativeWidth=1.5,PositiveHeight=1,NegativeHeight=1.5)
    FalloffSize=0.2
    FlipHeightScaling=False
    FlipWidthScaling=False
    OptionalParentItem=(Type=Bone,Name="None")
    RotationOffset=(X=0,Y=130,Z=0)
    ExecutePin=@RigUnit_BeginExecution.ExecutePin
   )
   > &RigUnit_ModifyTransforms(
    ItemToModify=((Item=(Type=Bone,Name="dyn_feathers_01_corr_01"),Transform=(Rotation=(X=0,Y=0,Z=-0.3827,W=0.9239),Translation=(X=0,Y=-15,Z=0),Scale3D=(X=1,Y=1,Z=1))))
    WeightMinimum=0
    Weight=@n1.OutputParam
   )
   > &RigUnit_ModifyTransforms(
    ItemToModify=((Item=(Type=Bone,Name="dyn_feathers_02_corr_02"),Transform=(Rotation=(X=0,Y=0,Z=0,W=1),Translation=(X=0,Y=-13,Z=0),Scale3D=(X=1,Y=1,Z=1))))
    WeightMinimum=0
    Weight=@n1.OutputParam
   )
   > &RigUnit_ModifyTransforms(
    ItemToModify=((Item=(Type=Bone,Name="dyn_feathers_01_corr_02"),Transform=(Rotation=(X=-0.1305,Y=0,Z=0,W=0.9914),Translation=(X=1,Y=-3,Z=0),Scale3D=(X=1,Y=1,Z=1))))
    WeightMinimum=0
    Weight=@n1.OutputParam
   )
   > n2 = &RigUnit_SphericalPoseReader(
    ActiveRegionScaleFactors=(PositiveWidth=0.8,NegativeWidth=0.8,PositiveHeight=0.8,NegativeHeight=0.8)
    ActiveRegionSize=0.1
    Debug=(bDrawDebug=false,bDraw2D=False,bDrawLocalAxes=False,DebugScale=25,DebugSegments=20,DebugThickness=0.25)
    DriverAxis=(X=1,Y=0,Z=0)
    DriverItem=(Type=Bone,Name="upperarm_l")
    FalloffRegionScaleFactors=(PositiveWidth=1,NegativeWidth=1,PositiveHeight=1,NegativeHeight=1)
    FalloffSize=0.2
    FlipHeightScaling=False
    FlipWidthScaling=False
    OptionalParentItem=(Type=Bone,Name="None")
    RotationOffset=(X=-10,Y=-45,Z=0)
   )
   n3 = &RigVMFunction_MathDoubleSub(A=@n2.OutputParam, B=RigVMVariableNode() {variable=IsParachuteOpened}.Value)
   > &RigUnit_ModifyTransforms(
    ItemToModify=((Item=(Type=Bone,Name="dyn_feathers_03_corr_03"),Transform=(Rotation=(X=0,Y=0,Z=-0.4617,W=0.887),Translation=(X=1,Y=-13,Z=0),Scale3D=(X=1,Y=1,Z=1))))
    WeightMinimum=0
    Weight=@n3.Result
   )
   > &RigUnit_ModifyTransforms(
    ItemToModify=((Item=(Type=Bone,Name="dyn_feathers_02_corr_03"),Transform=(Rotation=(X=0,Y=0,Z=0,W=1),Translation=(X=0,Y=-20,Z=0),Scale3D=(X=1,Y=1,Z=1))))
    WeightMinimum=0
    Weight=@n3.Result
   )
   > n4 = &RigUnit_SphericalPoseReader(
    ActiveRegionScaleFactors=(PositiveWidth=0.8,NegativeWidth=0.8,PositiveHeight=0.8,NegativeHeight=0.8)
    ActiveRegionSize=0.1
    Debug=(bDrawDebug=false,bDraw2D=False,bDrawLocalAxes=False,DebugScale=25,DebugSegments=20,DebugThickness=0.25)
    DriverAxis=(X=1,Y=0,Z=0)
    DriverItem=(Type=Bone,Name="upperarm_l")
    FalloffRegionScaleFactors=(PositiveWidth=1,NegativeWidth=1,PositiveHeight=1,NegativeHeight=1)
    FalloffSize=0.2
    FlipHeightScaling=False
    FlipWidthScaling=False
    OptionalParentItem=(Type=Bone,Name="None")
    RotationOffset=(X=0,Y=-60,Z=0)
   )
   n5 = &RigVMFunction_MathDoubleMul(A=@n4.OutputParam, B=RigVMVariableNode() {variable=IsParachuteOpened}.Value)
   > &RigUnit_ModifyTransforms(
    ItemToModify=((Item=(Type=Bone,Name="dyn_feathers_03_corr_03"),Transform=(Rotation=(X=-0.2468,Y=0.0778,Z=-0.2905,W=0.9212),Translation=(X=1,Y=-11,Z=0),Scale3D=(X=1,Y=1,Z=1))))
    WeightMinimum=0
    Weight=@n5.Result
   )
   > &RigUnit_ModifyTransforms(
    ItemToModify=((Item=(Type=Bone,Name="dyn_feathers_02_corr_03"),Transform=(Rotation=(X=0,Y=0,Z=0,W=1),Translation=(X=0,Y=-20,Z=0),Scale3D=(X=1,Y=1,Z=1))))
    WeightMinimum=0
    Weight=@n5.Result
   )
  ]}
]
hierarchy [
 root {loc=<0 0 0>, rot=<0 0 0 1>}
 pelvis {parent=root, loc=<0 -0.5301 98.0571>, rot=<0.0127 0.707 -0.0127 -0.707>}
 spine_01 {parent=pelvis, loc=<6.1208 -0.9328 0>, rot=<0 0 0.034 -0.9994>}
 spine_02 {parent=spine_01, loc=<6.5996 -0.2095 0>, rot=<0 0 0.0238 -0.9997>}
 spine_03 {parent=spine_02, loc=<6.4914 -0.1237 0>, rot=<0 0 0.0149 -0.9999>}
 spine_04 {parent=spine_03, loc=<6.9598 0.0794 0>, rot=<0 0 -0.0132 -0.9999>}
 spine_05 {parent=spine_04, loc=<6.9987 0 0>, rot=<0 0 -0.0015 -1>}
 attach_cape {parent=spine_05, loc=<9.0184 -11.7141 -0.0001>, rot=<0.0616 -0.7044 -0.0616 -0.7044>}
 neck_01 {parent=spine_05, loc=<11.3138 0.1529 0>, rot=<0 0 -0.2003 -0.9797>}
 neck_02 {parent=neck_01, loc=<4.5627 0 0>, rot=<0 0 0.0504 -0.9987>}
 head {parent=neck_02, loc=<5.3914 0 0>, rot=<0 0 0.115 -0.9934>}
 dyn_feathers_01 {parent=head, loc=<0 0 0>, rot=<0 0 0 1>}
 dyn_feathers_01_corr_01 {parent=dyn_feathers_01, loc=<0.5905 3.1514 17.915>, rot=<-0.6397 -0.0079 0.0094 -0.7685>}
 dyn_feathers_01_corr_02 {parent=dyn_feathers_01, loc=<4.9432 -8.3475 10.4268>, rot=<0.8804 0.1714 -0.0845 0.4341>}
 dyn_feathers_01_corr_03 {parent=dyn_feathers_01, loc=<4.0097 -11.3143 -9.8192>, rot=<0.9326 0.1171 0.0425 -0.3388>}
 dyn_feathers_01_corr_04 {parent=dyn_feathers_01, loc=<-7.6016 18.3919 -15.1533>, rot=<0.332 -0.0149 -0.1613 -0.9293>}
 dyn_feathers_02 {parent=head, loc=<0 0 0>, rot=<0 0 0 1>}
 dyn_feathers_02_corr_01 {parent=dyn_feathers_02, loc=<-4.2336 11.8684 17.7507>, rot=<-0.4405 0.0288 -0.0586 -0.8954>}
 dyn_feathers_02_corr_02 {parent=dyn_feathers_02, loc=<1.3381 -3.2808 14.7375>, rot=<-0.7821 -0.0201 0.0349 -0.6218>}
 dyn_feathers_02_corr_03 {parent=dyn_feathers_02, loc=<2.892 -3.8315 -16.5792>, rot=<0.7723 0.061 0.0498 -0.6303>}
 dyn_feathers_02_corr_04 {parent=dyn_feathers_02, loc=<-4.2308 11.6448 -17.6637>, rot=<0.4698 -0.0488 -0.0491 -0.88>}
 dyn_feathers_03 {parent=head, loc=<0 0 0>, rot=<0 0 0 1>}
 dyn_feathers_03_corr_01 {parent=dyn_feathers_03, loc=<-7.485 18.763 16.2766>, rot=<-0.4067 0.0622 -0.1377 -0.901>}
 dyn_feathers_03_corr_02 {parent=dyn_feathers_03, loc=<5.1729 -13.1932 -1.9293>, rot=<0.9797 0.1844 0.0036 -0.0783>}
 dyn_feathers_03_corr_03 {parent=dyn_feathers_03, loc=<-0.0471 3.3328 -19.8708>, rot=<0.6405 0.0198 0.0259 -0.7673>}
 faceAttach {parent=head, loc=<-4.146 -0.8545 0>, rot=<0.0029 -0.7071 -0.0029 -0.7071>}
 attach_backpack {parent=spine_05, loc=<-0.1213 -13.4731 -0.0001>, rot=<-0.0093 -0.707 0.0093 -0.707>}
 clavicle_l {parent=spine_05, loc=<5.4831 -0.1053 -3.2946>, rot=<0.0465 -0.7289 -0.009 -0.6829>}
 upperarm_l {parent=clavicle_l, loc=<9.8658 -1.2136 -0.4105>, rot=<0.0164 -0.3245 -0.0269 -0.9454>}
 lowerarm_l {parent=upperarm_l, loc=<26.465 -0.8345 0.259>, rot=<0.0036 0.0161 -0.2298 -0.9731>}
 hand_l {parent=lowerarm_l, loc=<22.7355 -0.3203 -0.6105>, rot=<-0.6784 -0.0223 0.0735 -0.7307>}
 thumb_01_l {parent=hand_l, loc=<1.3118 -0.3191 -2.459>, rot=<0.6546 -0.4042 0.0664 -0.6354>}
 weapon_l {parent=hand_l, loc=<-0.1974 -3.4111 0.3811>, rot=<0 0 0 1>}
 lowerarm_twist_02_l {parent=lowerarm_l, loc=<10.6707 0 0>, rot=<0 0 0 1>}
 lowerarm_twist_01_l {parent=lowerarm_l, loc=<18.316 0 0>, rot=<0 0 0 1>}
 upperarm_twist_01_l {parent=upperarm_l, loc=<7.2135 -0.1134 0>, rot=<0 0 0 1>}
 upperarm_twist_02_l {parent=upperarm_l, loc=<19.7998 -0.1134 0>, rot=<0 0 0 1>}
 clavicle_r {parent=spine_05, loc=<5.4831 -0.1053 3.2946>, rot=<0.7289 0.0465 -0.6829 0.009>}
 upperarm_r {parent=clavicle_r, loc=<-9.8658 1.2136 0.4105>, rot=<0.0164 -0.3245 -0.0269 -0.9454>}
 lowerarm_r {parent=upperarm_r, loc=<-26.465 0.8345 -0.259>, rot=<0.0036 0.0161 -0.2298 -0.9731>}
 upperarm_twist_01_r {parent=upperarm_r, loc=<-7.2135 0.1134 0>, rot=<0 0 0 -1>}
 upperarm_twist_02_r {parent=upperarm_r, loc=<-19.7998 0.1134 0>, rot=<0 0 0 -1>}
 dfrm_clavicle_l {parent=spine_05, loc=<5.4831 -0.1053 -3.2946>, rot=<0.0465 -0.7289 -0.009 -0.6829>}
 dfrm_upperarm_l {parent=dfrm_clavicle_l, loc=<9.8658 -1.2136 -0.4105>, rot=<0.0164 -0.3245 -0.0269 -0.9454>}
 dfrm_lowerarm_l {parent=dfrm_upperarm_l, loc=<26.465 -0.8345 0.259>, rot=<0.0036 0.0161 -0.2298 -0.9731>}
 dfrm_lowerarm_twist_02_l {parent=dfrm_lowerarm_l, loc=<10.6707 0 0>, rot=<0 0 0 1>}
 dyn_wing_lower_l_twist_MSTR {parent=dfrm_lowerarm_twist_02_l, loc=<0 0 0>, rot=<0 0 0 1>}
 dyn_wing_lower_l_MSTR {parent=dyn_wing_lower_l_twist_MSTR, loc=<-7.658 -2.048 -0.4383>, rot=<0.0097 -0.1651 0.9668 0.1948>}
 dyn_wing_upper_l_04 {parent=dyn_wing_lower_l_MSTR, loc=<-1.1461 3.2663 -0.3853>, rot=<0.0433 -0.0165 -0.0185 -0.9988>}
 dyn_wing_upper_l_04_end {parent=dyn_wing_upper_l_04, loc=<-9.3847 0.0001 -0.0002>, rot=<0.0023 0.0263 0.0172 -0.9995>}
 dyn_wing_upper_l_05 {parent=dyn_wing_lower_l_MSTR, loc=<-10.9451 -1.8257 -0.4175>, rot=<0 0 -0.039 -0.9992>}
 dyn_wing_upper_l_05_end {parent=dyn_wing_upper_l_05, loc=<-8.0176 0 0>, rot=<-0.004 -0.0349 0.0426 -0.9985>}
 dfrm_lowerarm_twist_01_l {parent=dfrm_lowerarm_l, loc=<18.316 0 0>, rot=<0 0 0 1>}
 dfrm_hand_l {parent=dfrm_lowerarm_l, loc=<22.7355 -0.3203 -0.6105>, rot=<-0.6784 -0.0223 0.0735 -0.7307>}
 dfrm_upperarm_twist_01_l {parent=dfrm_upperarm_l, loc=<7.2135 -0.1134 0>, rot=<0.0015 0 0 -1>}
 dyn_wing_upper_l_twist_MSTR {parent=dfrm_upperarm_twist_01_l, loc=<0 0 0>, rot=<0 0 0 1>}
 dyn_wing_upper_l_MSTR {parent=dyn_wing_upper_l_twist_MSTR, loc=<2.858 -3.3049 -3.137>, rot=<-0.0599 -0.3671 0.8988 0.2321>}
 dyn_wing_upper_l_02 {parent=dyn_wing_upper_l_MSTR, loc=<-7.1189 1.3474 -0.9037>, rot=<0 0 0 1>}
 dyn_wing_upper_l_02_end {parent=dyn_wing_upper_l_02, loc=<-6.8685 0 0>, rot=<-0.0087 0.0178 -0.0208 -0.9996>}
 dyn_wing_upper_l_03 {parent=dyn_wing_upper_l_MSTR, loc=<-1.5405 4.0704 -0.5145>, rot=<0.0076 0.0232 0.1055 -0.9941>}
 dyn_wing_upper_l_03_end {parent=dyn_wing_upper_l_03, loc=<-6.0908 0 0>, rot=<-0.0222 0.0342 -0.0093 -0.9991>}
 dfrm_upperarm_twist_02_l {parent=dfrm_upperarm_l, loc=<19.7998 -0.1134 0>, rot=<0 0 0 1>}
 dfrm_clavicle_r {parent=spine_05, loc=<5.4831 -0.1053 3.2946>, rot=<0.7289 0.0465 -0.6829 0.009>}
 dfrm_upperarm_r {parent=dfrm_clavicle_r, loc=<-9.8658 1.2136 0.4105>, rot=<0.0164 -0.3245 -0.0269 -0.9454>}
 dfrm_upperarm_twist_01_r {parent=dfrm_upperarm_r, loc=<-7.2135 0.1134 0>, rot=<0 0 0 -1>}
 dyn_wing_upper_r_twist_MSTR {parent=dfrm_upperarm_twist_01_r, loc=<0 0 0>, rot=<0 0 0 1>}
 dyn_wing_upper_r_MSTR {parent=dyn_wing_upper_r_twist_MSTR, loc=<-2.8583 3.3049 3.1373>, rot=<-0.0599 -0.3671 0.8988 0.2321>}
 dyn_wing_upper_r_02 {parent=dyn_wing_upper_r_MSTR, loc=<7.1186 -1.3473 0.9034>, rot=<0 0 0 1>}
 dyn_wing_upper_r_02_end {parent=dyn_wing_upper_r_02, loc=<6.8685 0 0>, rot=<-0.0087 0.0178 -0.0208 -0.9996>}
 dyn_wing_upper_r_03 {parent=dyn_wing_upper_r_MSTR, loc=<1.5403 -4.0703 0.5142>, rot=<0.0076 0.0232 0.1055 -0.9941>}
 dyn_wing_upper_r_03_end {parent=dyn_wing_upper_r_03, loc=<6.0908 0 0>, rot=<-0.0222 0.0342 -0.0093 -0.9991>}
 dfrm_upperarm_twist_02_r {parent=dfrm_upperarm_r, loc=<-19.7998 0.1134 0>, rot=<0 0 0 -1>}
 dyn_feather_R_2_MSTR {parent=spine_05, loc=<12.4086 1.5703 9.7044>, rot=<-0.2357 0.3114 -0.1375 -0.9102>}
 dyn_feather_R_2_01 {parent=dyn_feather_R_2_MSTR, loc=<0 0 0>, rot=<0 0 0 1>}
 dyn_feather_R_2_02 {parent=dyn_feather_R_2_01, loc=<9.4479 0 0>, rot=<0 0 0 1>}
 dyn_feather_R_2_03 {parent=dyn_feather_R_2_02, loc=<3.519 3.9135 -0.8141>, rot=<0.0842 0.1874 -0.1546 -0.9664>}
 dyn_feather_R_2_04 {parent=dyn_feather_R_2_02, loc=<3.6296 -3.6423 -0.7291>, rot=<-0.1415 0.0893 0.0965 -0.9812>}
 dyn_feather_back_MSTR {parent=spine_05, loc=<15.7887 -5.3233 2.2789>, rot=<-0.6177 0.2159 0.0743 -0.7525>}
 dyn_feather_back_01 {parent=dyn_feather_back_MSTR, loc=<0 0 0>, rot=<0 0 0 1>}
 dyn_feather_back_02 {parent=dyn_feather_back_01, loc=<6.814 0 0>, rot=<0 0 0 1>}
 dyn_feather_back_03 {parent=dyn_feather_back_02, loc=<3.7643 -5.1494 -0.7805>, rot=<-0.1594 -0.1304 0.0628 -0.9766>}
 dyn_feather_back_04 {parent=dyn_feather_back_02, loc=<2.1898 5.7367 -2.8405>, rot=<0.1259 -0.1046 -0.123 -0.9788>}
 dyn_feather_L_2_MSTR {parent=spine_05, loc=<13.826 -0.6581 -9.2305>, rot=<0.9342 0.0817 -0.3104 0.1559>}
 dyn_feather_L_2_01 {parent=dyn_feather_L_2_MSTR, loc=<0 0 0>, rot=<0 0 0 1>}
 dyn_feather_L_2_02 {parent=dyn_feather_L_2_01, loc=<6.814 0 0>, rot=<0 0 0 1>}
 dyn_feather_L_2_03 {parent=dyn_feather_L_2_02, loc=<2.7467 -7.4853 1.8355>, rot=<0.0675 0.1003 0.125 -0.9848>}
 dyn_feather_L_2_04 {parent=dyn_feather_L_2_02, loc=<2.6022 8.84 -3.0731>, rot=<0.1714 0.0581 -0.1591 -0.9705>}
 dyn_feather_L_2_05 {parent=dyn_feather_L_2_02, loc=<5.5861 -0.0021 -0.899>, rot=<0.1641 0.0765 -0.0522 -0.9821>}
 dyn_feather_L_MSTR {parent=spine_05, loc=<10.0324 11.6845 -7.7702>, rot=<0.816 0.2725 -0.4884 -0.1462>}
 dyn_feather_L_01 {parent=dyn_feather_L_MSTR, loc=<0 0 0>, rot=<0 0 0 1>}
 dyn_feather_L_02 {parent=dyn_feather_L_01, loc=<6.814 0 0>, rot=<0 0 0 1>}
 dyn_feather_L_03 {parent=dyn_feather_L_02, loc=<4.1489 4.1168 -0.7032>, rot=<0.2833 0.0762 -0.1513 -0.944>}
 dyn_feather_L_04 {parent=dyn_feather_L_02, loc=<3.4144 -3.6889 1.2811>, rot=<0.1844 0.1658 -0.0098 -0.9687>}
 dyn_feather_R_MSTR {parent=spine_05, loc=<8.6124 11.004 7.5754>, rot=<0.1031 0.3982 -0.2984 -0.8613>}
 dyn_feather_R_01 {parent=dyn_feather_R_MSTR, loc=<0 0 0>, rot=<0 0 0 1>}
 dyn_feather_R_02 {parent=dyn_feather_R_01, loc=<7.799 0 0>, rot=<0 0 0 1>}
 dyn_feather_R_03 {parent=dyn_feather_R_02, loc=<5.0341 -4.0277 1.6993>, rot=<-0.3465 0.3473 0.0453 -0.8702>}
 dyn_feather_R_04 {parent=dyn_feather_R_02, loc=<6.2265 3.5994 3.6298>, rot=<-0.1029 0.4875 -0.0396 -0.8661>}
 dyn_scarf_MSTR {parent=spine_05, loc=<13.965 -6.5288 0.5485>, rot=<0.0453 -0.67 -0.0094 -0.7409>}
 dyn_scarf_01 {parent=dyn_scarf_MSTR, loc=<0 0 0>, rot=<0 0 0 1>}
 dyn_scarf_02 {parent=dyn_scarf_01, loc=<2.3123 0 0>, rot=<-0.4821 0.4922 -0.2213 -0.6902>}
 dyn_scarf_03 {parent=dyn_scarf_01, loc=<-2.5064 -0.2643 0.4758>, rot=<0.2862 0.6173 0.2268 -0.6969>}
 dfrm_chest_01_l {parent=spine_04, loc=<0.5425 15.2169 -7.801>, rot=<-0.5908 -0.3886 0.621 0.3381>}
 dfrm_chest_01_r {parent=spine_04, loc=<2.3593 15.1631 6.5797>, rot=<0.3475 0.6186 -0.395 -0.5835>}
 thigh_l {parent=pelvis, loc=<-1.0855 -0.2109 -8.7053>, rot=<0.0009 0.0355 0.0155 -0.9993>}
 calf_l {parent=thigh_l, loc=<-42.5284 -0.0019 -0.7297>, rot=<0.0006 -0.0161 -0.0391 -0.9991>}
 foot_l {parent=calf_l, loc=<-46.7401 -2.5014 -0.592>, rot=<0.0427 -0.0238 0.0353 -0.9982>}
 ball_l {parent=foot_l, loc=<-3.5658 10.8066 -0.1544>, rot=<0.0001 -0.0001 -0.7063 -0.7079>}
 dfrm_foot_aim_l {parent=foot_l, loc=<-4.054 -7.761 -1.1648>, rot=<0 0 0 -1>}
 calf_twist_02_l {parent=calf_l, loc=<-19.1546 -1.7943 -0.2638>, rot=<0.0237 0.0095 -0.0035 -0.9997>}
 calf_twist_01_l {parent=calf_l, loc=<-35.8483 -1.9197 -0.5794>, rot=<0.0237 0.0095 -0.0035 -0.9997>}
 thigh_r {parent=pelvis, loc=<-1.0855 -0.2109 8.7053>, rot=<0.0355 -0.0009 0.9993 0.0155>}
 calf_r {parent=thigh_r, loc=<42.5284 0.0019 0.7297>, rot=<0.0006 -0.0161 -0.0391 -0.9991>}
 foot_r {parent=calf_r, loc=<46.7401 2.5014 0.592>, rot=<0.0427 -0.0238 0.0353 -0.9982>}
 ball_r {parent=foot_r, loc=<3.5658 -10.8066 0.1544>, rot=<0.0001 -0.0001 -0.7063 -0.7079>}
 dfrm_foot_aim_r {parent=foot_r, loc=<4.054 7.761 1.1648>, rot=<0 0 0 -1>}
 calf_twist_02_r {parent=calf_r, loc=<19.1546 1.7943 0.2638>, rot=<0.0237 0.0095 -0.0035 -0.9997>}
 calf_twist_01_r {parent=calf_r, loc=<35.8483 1.9197 0.5794>, rot=<0.0237 0.0095 -0.0035 -0.9997>}
 thigh_twist_01_r {parent=thigh_r, loc=<27.952 0 0>, rot=<0.0013 0 0 -1>}
 dfrm_pelvis {parent=pelvis, loc=<-12.6365 0 0>, rot=<0 0 0 1>}
 dfrm_thigh_r {parent=dfrm_pelvis, loc=<-3.2425 0.2481 8.7122>, rot=<0.0572 -0.0012 0.9982 0.0155>}
 dfrm_calf_r {parent=dfrm_thigh_r, loc=<37.1895 -0.1455 0.06>, rot=<0.0006 -0.0161 -0.0391 -0.9991>}
 dfrm_calf_twist_01_r {parent=dfrm_calf_r, loc=<29.2812 0.7545 -0.3098>, rot=<0.0237 0.0095 -0.0035 -0.9997>}
 dfrm_calf_twist_02_r {parent=dfrm_calf_r, loc=<12.2745 0.7545 -0.3098>, rot=<0.0237 0.0095 -0.0035 -0.9997>}
 dyn_calf_feather_01 {parent=dfrm_calf_twist_02_r, loc=<-2.0783 -2.7413 -0.5344>, rot=<-0.724 -0.0054 0.0675 -0.6865>}
 dyn_calf_feather_02 {parent=dfrm_calf_twist_02_r, loc=<-7.448 -2.9439 -1.5042>, rot=<-0.8183 0.0047 0.0676 -0.5708>}
 dyn_calf_feather_03 {parent=dfrm_calf_twist_02_r, loc=<-2.7347 0.5821 3.7067>, rot=<-0.0446 -0.0508 0.0449 -0.9967>}
 dyn_calf_feather_04 {parent=dfrm_calf_twist_02_r, loc=<-8.4089 0.7852 3.5059>, rot=<-0.0446 -0.0508 0.0449 -0.9967>}
 dyn_calf_feather_05 {parent=dfrm_calf_twist_02_r, loc=<-8.3312 4.92 0.8996>, rot=<0.6751 -0.0676 -0.0043 -0.7346>}
 dyn_calf_feather_06 {parent=dfrm_calf_twist_02_r, loc=<-3.1964 4.7196 2.2151>, rot=<0.4667 -0.0665 0.013 -0.8818>}
 dyn_calf_feather_07 {parent=dfrm_calf_twist_02_r, loc=<-2.5695 2.1758 -3.1154>, rot=<0.9841 -0.0506 -0.045 -0.1639>}
 dyn_calf_feather_08 {parent=dfrm_calf_twist_02_r, loc=<-7.5771 2.7904 -3.1329>, rot=<0.9841 -0.0506 -0.045 -0.1639>}
 deform_knee_in_base_r {parent=dfrm_calf_r, loc=<0.3084 7.9762 -0.8686>, rot=<0 0 0 -1>}
 deform_knee_in_r {parent=deform_knee_in_base_r, loc=<0 0 0>, rot=<0 0 0 1>}
 deform_knee_out_base_r {parent=dfrm_calf_r, loc=<-0.8551 -8.4142 -0.7848>, rot=<0 0 0 -1>}
 deform_knee_out_r {parent=deform_knee_out_base_r, loc=<0 0 0>, rot=<0 0 0 1>}
 dfrm_foot_r {parent=dfrm_calf_r, loc=<42.065 -0.9449 -1.8502>, rot=<-0.0025 -0.0618 -0.6802 -0.7304>}
 dyn_ball_r {parent=dfrm_foot_r, loc=<-5.8402 -1.077 0.421>, rot=<0 0 0 1>}
 dyn_toe_r_c_01 {parent=dyn_ball_r, loc=<0.5132 -0.201 3.0849>, rot=<-0.2023 -0.6858 -0.6661 0.2124>}
 dyn_toe_r_c_02 {parent=dyn_toe_r_c_01, loc=<3.8303 0 0>, rot=<0 -0.0255 0.0482 -0.9985>}
 dyn_toe_r_a_01 {parent=dyn_ball_r, loc=<0.3668 -0.1038 -2.6579>, rot=<-0.156 0.673 0.6981 0.188>}
 dyn_toe_r_a_02 {parent=dyn_toe_r_a_01, loc=<3.64 0 0>, rot=<0 -0.0797 -0.0135 -0.9967>}
 dyn_toe_r_d_03 {parent=dfrm_foot_r, loc=<2.6964 0.0647 0.4623>, rot=<0.7071 -0.1005 0.0447 -0.6985>}
 dyn_toe_r_d_01 {parent=dyn_toe_r_d_03, loc=<2.0486 0 0>, rot=<-0.0333 -0.0352 0.0011 -0.9988>}
 dyn_toe_r_d_02 {parent=dyn_toe_r_d_01, loc=<1.8873 0 0>, rot=<0 0.066 0 -0.9978>}
 dfrm_thigh_twist_01_r {parent=dfrm_thigh_r, loc=<13.2027 0 0>, rot=<0.0013 0 0 -1>}
 dfrm_thigh_r_lod_corr {parent=dfrm_thigh_r, loc=<0 0 0>, rot=<0 0 0 1>}
 dfrm_thigh_l {parent=dfrm_pelvis, loc=<-3.2425 0.2481 -8.7122>, rot=<0.0012 0.0572 0.0155 -0.9982>}
 dfrm_calf_l {parent=dfrm_thigh_l, loc=<-37.1895 0.1455 -0.06>, rot=<0.0006 -0.0161 -0.0391 -0.9991>}
 dfrm_calf_twist_01_l {parent=dfrm_calf_l, loc=<-29.2811 -0.7545 0.3098>, rot=<0.0237 0.0095 -0.0035 -0.9997>}
 dfrm_calf_twist_02_l {parent=dfrm_calf_l, loc=<-12.2744 -0.7545 0.3098>, rot=<0.0237 0.0095 -0.0035 -0.9997>}
 dyn_calf_feather_l_09 {parent=dfrm_calf_twist_02_l, loc=<3.3487 -0.2994 -3.7679>, rot=<-0.0528 -0.9746 -0.2135 0.0424>}
 dyn_calf_feather_l_07 {parent=dfrm_calf_twist_02_l, loc=<3.903 -4.7153 -0.8825>, rot=<-0.0042 -0.8134 0.5777 0.0676>}
 dyn_calf_feather_l_05 {parent=dfrm_calf_twist_02_l, loc=<3.4028 -2.4962 2.7341>, rot=<0.0396 -0.2639 0.9622 0.055>}
 dyn_calf_feather_l_03 {parent=dfrm_calf_twist_02_l, loc=<3.7433 0.9814 3.1347>, rot=<0.0596 0.1812 0.9811 0.0323>}
 dyn_calf_feather_l_02 {parent=dfrm_calf_twist_02_l, loc=<3.5916 2.5427 0.4456>, rot=<0.0675 0.565 0.8223 0.0052>}
 dfrm_foot_l {parent=dfrm_calf_l, loc=<-42.065 0.9449 1.8503>, rot=<-0.0025 -0.0618 -0.6802 -0.7304>}
 dyn_ball_l {parent=dfrm_foot_l, loc=<5.8402 1.077 -0.421>, rot=<0 0 0 1>}
 dyn_toe_l_c_01 {parent=dyn_ball_l, loc=<-0.3668 0.1038 2.6579>, rot=<-0.6896 0.1883 -0.1422 -0.6847>}
 dyn_toe_l_c_02 {parent=dyn_toe_l_c_01, loc=<5.0037 0 0>, rot=<0.0036 -0.1325 -0.0096 -0.9911>}
 dyn_toe_l_a_01 {parent=dyn_ball_l, loc=<-0.4612 0.1551 -3.2374>, rot=<-0.6727 -0.187 0.1986 -0.6878>}
 dyn_toe_l_a_02 {parent=dyn_toe_l_a_01, loc=<4.8931 0 0>, rot=<-0.1023 -0.0736 -0.0192 -0.9918>}
 dyn_toe_l_d_03 {parent=dfrm_foot_l, loc=<-2.6963 -0.0647 -0.4626>, rot=<-0.0696 -0.6943 0.7053 0.1251>}
 dyn_toe_l_d_01 {parent=dyn_toe_l_d_03, loc=<2.0436 0 0.1444>, rot=<0 0 0 1>}
 dyn_toe_l_d_02 {parent=dyn_toe_l_d_01, loc=<1.8873 0 0>, rot=<-0.0107 0.0658 -0.0051 -0.9978>}
 dfrm_stomach_01_r {parent=pelvis, loc=<5.3295 19.3732 7.4074>, rot=<0.7074 0.6825 0.1322 -0.1275>}
 dfrm_stomach_02_r {parent=pelvis, loc=<5.6777 12.2012 15.2478>, rot=<0.584 0.5634 0.4206 -0.4058>}
 dfrm_stomach_02_l {parent=pelvis, loc=<5.6777 12.2012 -15.248>, rot=<0.6319 0.6096 -0.3445 0.3323>}
 dfrm_stomach_01_l {parent=pelvis, loc=<5.3295 19.3732 -7.407>, rot=<0.7041 0.6793 -0.1487 0.1435>}
 ik_hand_root {parent=root, loc=<0 0 0>, rot=<0 0 0 1>}
 ik_hand_gun {parent=ik_hand_root, loc=<-48.1743 3.3545 104.4381>, rot=<0.6276 0.1309 0.4183 -0.6434>}
 ik_hand_l {parent=ik_hand_gun, loc=<59.3271 67.6997 34.3524>, rot=<-0.6158 -0.7027 -0.3565 0>}
 ik_foot_root {parent=root, loc=<0 0 0>, rot=<0 0 0 1>}
 ik_foot_l {parent=ik_foot_root, loc=<14.8262 -7.3545 8.296>, rot=<0.0357 0.7045 0.0284 -0.7082>}
 ik_foot_r {parent=ik_foot_root, loc=<-14.8262 -7.3545 8.296>, rot=<0.7082 0.0284 -0.7045 0.0357>}
]
variables [IsParachuteOpened {type=double, default=0}]
