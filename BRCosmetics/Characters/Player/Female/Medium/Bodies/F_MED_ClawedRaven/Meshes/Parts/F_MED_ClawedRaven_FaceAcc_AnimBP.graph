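# Anim Blueprint for the ClawedRaven female face accessory, animating its feathers and scarf.
# Parent pose feeds a parachute-aware Control Rig, spring-driven AnimDynamics, then corrective bone offsets, including a scarf offset when backbling is worn.
# Updates smooth parachute state, compute gravity override and derive FeatherFix from right-arm yaw or targeting.
# ShoeMesh presence switches skin weight profiles and the shoeScale morph, gated by mutually resetting DoOnce nodes.
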AnimBlueprint /BRCosmetics/Characters/Player/Female/Medium/Bodies/F_MED_ClawedRaven/Meshes/Parts/F_MED_ClawedRaven_FaceAcc_AnimBP
target_skeleton=&/F_MED_ClawedRaven_FaceAcc_Skeleton
graphs [
 AnimGraph {nodes [
   n1 = "Transform (Modify) Bone - Bone: dyn_feathers_02_corr_01"!(ComponentPose, Alpha=1, Translation="-1.436,0.336,0.495", Rotation="0,0,0") -> (Pose) {
    rotationSpace=BCS_BoneSpace
    translationMode=BMM_Additive
    translationSpace=BCS_BoneSpace
    alphaBoolBlend {blendOption=Linear}
    alphaScaleBiasClamp {bInterpResult=true, interpSpeedDecreasing=5, interpSpeedIncreasing=20}
    boneToModify {boneName=dyn_feathers_02_corr_01}
    translation=<-1.436 0.336 0.495>
   }
   "Transform (Modify) Bone - Bone: dyn_feathers_03_corr_01"!(ComponentPose=@n1, Alpha=1, Translation="-0.608,-6.131,-5.286", Rotation="0,0,0") -> (Pose) {
    rotationSpace=BCS_BoneSpace
    translationMode=BMM_Additive
    translationSpace=BCS_BoneSpace
    alphaBoolBlend {blendOption=Linear}
    alphaScaleBiasClamp {bInterpResult=true, interpSpeedDecreasing=5, interpSpeedIncreasing=20}
    boneToModify {boneName=dyn_feathers_03_corr_01}
    translation=<-0.608 -6.131 -5.286>
   }
   n2 = Control Rig!(
    Source=Copy Pose From Mesh() {bUseAttachedParent=true}
    Alpha
    IsParachuteOpened:double=Get Is Parachute Opened() -> (Is Parachute Opened:double)
   ) -> (Pose) {
    alphaBoolBlend {blendOption=Linear}
    controlRigAssetReference {blueprintRigClass=&/F_MED_ClawedRaven_FaceAcc_CtrlRig.F_MED_ClawedRaven_FaceAcc_CtrlRig_C}
   }
   n3 = Anim Dynamics - Bone: dyn_feathers_01!(ComponentPose=Local To Component(LocalPose=@n2), Alpha=0.15, GravityOverride:Vector="0,0,0") -> (Pose) {
    angularSpringConstant=80
    bAngularSpring=true
    bGravityOverrideInSimSpace=true
    bLinearSpring=true
    bOverrideLinearDamping=true
    bUseGravityOverride=true
    lODThreshold=1
    linearDampingOverride=0.99
    linearSpringConstant=80
    simulationSpace=Actor
    alphaBoolBlend {blendOption=Linear}
    boundBone {boneName=dyn_feathers_01}
    physicsBodyDefinitions [{
      sphereCollisionRadius=0
      boundBone {boneName=dyn_feathers_01}
      boxExtents=<30 30 30>
      constraintSetup {
       angularTargetAxis=AxisY
       angularLimitsMax=<0 30 30>
       angularLimitsMin=<0 -30 -30>
       angularTarget=<0 0 0>
       linearAxesMax=<10 20 5>
       linearAxesMin=<0 0 -20>
      }
      localJointOffset=<0 0 0>
     }]
   }
   n4 = Anim Dynamics - Bone: dyn_feathers_02!(ComponentPose=@n3, Alpha=0.2, GravityOverride:Vector="0,0,0") -> (Pose) {
    angularSpringConstant=80
    bAngularSpring=true
    bGravityOverrideInSimSpace=true
    bLinearSpring=true
    bOverrideLinearDamping=true
    bUseGravityOverride=true
    lODThreshold=1
    linearDampingOverride=0.99
    linearSpringConstant=80
    simulationSpace=Actor
    alphaBoolBlend {blendOption=Linear}
    boundBone {boneName=dyn_feathers_02}
    physicsBodyDefinitions [{
      sphereCollisionRadius=0
      boundBone {boneName=dyn_feathers_02}
      boxExtents=<30 30 30>
      constraintSetup {
       angularTargetAxis=AxisY
       angularLimitsMax=<0 30 30>
       angularLimitsMin=<0 -30 -30>
       angularTarget=<0 0 0>
       linearAxesMax=<10 20 20>
       linearAxesMin=<0 -5 -5>
      }
      localJointOffset=<0 -0.5 0>
     }]
   }
   n5 = Anim Dynamics - Bone: dyn_feathers_03!(ComponentPose=@n4, Alpha=0.25, GravityOverride:Vector="0,0,0") -> (Pose) {
    angularSpringConstant=80
    bAngularSpring=true
    bGravityOverrideInSimSpace=true
    bLinearSpring=true
    bOverrideLinearDamping=true
    bUseGravityOverride=true
    lODThreshold=1
    linearDampingOverride=0.99
    linearSpringConstant=80
    simulationSpace=Actor
    alphaBoolBlend {blendOption=Linear}
    boundBone {boneName=dyn_feathers_03}
    physicsBodyDefinitions [{
      sphereCollisionRadius=0
      boundBone {boneName=dyn_feathers_03}
      boxExtents=<30 30 30>
      constraintSetup {
       angularTargetAxis=AxisY
       angularLimitsMax=<0 30 30>
       angularLimitsMin=<0 -30 -30>
       angularTarget=<0 0 0>
       linearAxesMax=<10 10 10>
       linearAxesMin=<0 0 -10>
      }
      localJointOffset=<0 0 0>
     }]
   }
   n6 = Anim Dynamics - Bone: dyn_scarf_02!(ComponentPose=@n5, Alpha=1, GravityOverride:Vector="0,0,-980") -> (Pose) {
    angularBiasOverride=0.9999
    angularDampingOverride=0.9999
    angularSpringConstant=500
    bAngularSpring=true
    bGravityOverrideInSimSpace=true
    bOverrideAngularBias=true
    bOverrideAngularDamping=true
    bUseGravityOverride=true
    gravityScale=0.2
    simulationSpace=RootRelative
    alphaBoolBlend {blendOption=Linear}
    boundBone {boneName=dyn_scarf_02}
    componentAppliedLinearAccClamp=<3000 3000 3000>
    componentLinearAccScale=<0.5 0.5 0.5>
    componentLinearVelScale=<0.7 0.7 0.7>
    gravityOverride {z=-980}
    physicsBodyDefinitions [{
      collisionType=CustomSphere
      sphereCollisionRadius=1.5
      boundBone {boneName=dyn_scarf_02}
      boxExtents=<10 10 10>
      constraintSetup {angularTargetAxis=AxisY, angularLimitsMax=<20 0 30>, angularLimitsMin=<-20 0 -10>, angularTarget=<0 1 0>}
      localJointOffset=<-0.7 0 0>
     }]
    relativeSpaceBone {boneName=spine_05}
   }
   n7 = Anim Dynamics - Bone: dyn_scarf_03!(ComponentPose=@n6, Alpha=1, GravityOverride:Vector="0,0,-980") -> (Pose) {
    angularBiasOverride=0.9999
    angularDampingOverride=0.9999
    angularSpringConstant=500
    bAngularSpring=true
    bGravityOverrideInSimSpace=true
    bOverrideAngularBias=true
    bOverrideAngularDamping=true
    bUseGravityOverride=true
    gravityScale=0.2
    simulationSpace=RootRelative
    alphaBoolBlend {blendOption=Linear}
    boundBone {boneName=dyn_scarf_03}
    componentAppliedLinearAccClamp=<3000 3000 3000>
    componentLinearAccScale=<0.7 0.7 0.7>
    componentLinearVelScale=<0.9 0.9 0.9>
    gravityOverride {z=-980}
    physicsBodyDefinitions [{
      collisionType=CustomSphere
      sphereCollisionRadius=1.5
      boundBone {boneName=dyn_scarf_03}
      boxExtents=<8 15 20>
      constraintSetup {angularTargetAxis=AxisY, angularLimitsMax=<45 5 35>, angularLimitsMin=<-45 -5 -35>, angularTarget=<0 1 0>}
      localJointOffset=<-1 -0.1 0>
     }]
    relativeSpaceBone {boneName=spine_05}
   }
   n8 = "Transform (Modify) Bone - Bone: dyn_feathers_01_corr_01"(ComponentPose=@n7, Alpha=1, Translation="0.000,-3.323,-0.000", Rotation="0,0,0", Scale="0.5,0.5,0.5") {
    rotationMode=BMM_Additive
    rotationSpace=BCS_BoneSpace
    scaleMode=BMM_Additive
    scaleSpace=BCS_BoneSpace
    translationMode=BMM_Additive
    translationSpace=BCS_BoneSpace
    alphaBoolBlend {blendInTime=0.2, blendOption=Linear, blendOutTime=0.6}
    boneToModify {boneName=dyn_feathers_01_corr_01}
    scale=<0.5 0.5 0.5>
    translation {y=-3.323}
   }
   n9 = "Transform (Modify) Bone - Bone: dyn_feathers_02_corr_01"!(ComponentPose=@n8, Alpha=1, Translation="0.168,-5.836,5.165", Rotation="47.916,-40.866,29.386", Scale:Vector="0.5,0.5,0.5") -> (Pose) {
    rotationMode=BMM_Additive
    rotationSpace=BCS_BoneSpace
    scaleMode=BMM_Additive
    scaleSpace=BCS_BoneSpace
    translationMode=BMM_Additive
    translationSpace=BCS_BoneSpace
    alphaBoolBlend {blendInTime=0.2, blendOption=Linear, blendOutTime=0.6}
    boneToModify {boneName=dyn_feathers_02_corr_01}
    rotation=<rot 47.9161 -40.8662 29.3859>
    scale=<0.5 0.5 0.5>
    translation=<0.1679 -5.836 5.1648>
   }
   n10 = "Transform (Modify) Bone - Bone: dyn_feathers_03_corr_01"!(ComponentPose=@n9, Alpha=1, Translation="-2.483,-6.563,1.089", Rotation="0.000,-45.000,50.000", Scale:Vector="0.5,0.5,0.5") -> (Pose) {
    rotationMode=BMM_Additive
    rotationSpace=BCS_BoneSpace
    scaleMode=BMM_Additive
    scaleSpace=BCS_BoneSpace
    translationMode=BMM_Additive
    translationSpace=BCS_BoneSpace
    alphaBoolBlend {blendInTime=0.2, blendOption=Linear, blendOutTime=0.6}
    boneToModify {boneName=dyn_feathers_03_corr_01}
    rotation=<rot 0.0001 -45.0001 50>
    scale=<0.5 0.5 0.5>
    translation=<-2.4833 -6.5633 1.0886>
   }
   n11 = "Transform (Modify) Bone - Bone: dyn_scarf_02"!(ComponentPose=@n10, bAlphaBoolEnabled:bool=Get isWearingBackbling(), Translation="0,0,0", Rotation="0,20,-5", Scale="1,1,1") -> (Pose) {
    alphaInputType=Bool
    rotationMode=BMM_Additive
    rotationSpace=BCS_BoneSpace
    scaleSpace=BCS_BoneSpace
    translationMode=BMM_Additive
    translationSpace=BCS_BoneSpace
    alphaBoolBlend {blendInTime=0.2, blendOption=Linear, blendOutTime=0.2}
    boneToModify {boneName=dyn_scarf_02}
    rotation {roll=-5, yaw=20}
   }
   Output Pose(Result=Component To Local(ComponentPose=@n11)) {layerGroup=None}
  ]}
 EventGraph {nodes [
   n1 = Event BlueprintUpdateAnimation()
   > n2 = Sequence!(execute) -> (then_0, then_1, then_2:Exec, then_3:Exec)
   n3 = Get bIsParachuteOpened()
   > Set Is Parachute Opened(Is Parachute Opened:double=FInterpTo(Current=Get Is Parachute Opened(), Target="To Float (Boolean)"(InBool=@n3), DeltaTime=@n1.DeltaTimeX, InterpSpeed=10)) -> (Output_Get:double)
   n4 = BreakTransform(InTransform=GetDeltaTransformFromRefPose(self=Get BodyMesh() -> (BodyMesh:Skeletal Mesh Component Object Reference), BoneName=upperarm_r, BaseName=clavicle_r))
   n5 = Set upperarm r rot yaw(execute=@n2.then_1, upperarm r rot yaw:float=MapRangeClamped(Value=BreakRotator(InRot=@n4.Rotation).Yaw, InRangeA=50, InRangeB=90)) -> (Output_Get:float)
   > n6 = Gravity Override State Function DEPRECATED(CharacterPartAnimInstance=Self-Reference(), GameTimeInSeconds=GetGameTimeInSeconds(), GravityOverrideParams=Get Gravity Override Params())
   > Set Gravity Override(Gravity Override=@n6)
   n7 = Get bIsTargeting()
   > Set FeatherFix(FeatherFix:double="Max (Float)"(A=@n5, B="To Float (Boolean)"(InBool=@n7))) -> (Output_Get:double)
   n8 = Get!(execute:Exec=@n2.then_2) -> (then:Exec, else:Exec, ShoeMesh:Skeletal Mesh Component Budgeted Object Reference)
   > n9 = Sequence!(execute) -> (then_0, then_1)
   > Do Once(Reset=@n12.then_1)
   > n10 = Is Valid(InputObject=Get SkeletalMesh!(self:Skinned Mesh Component Object Reference=@n8) -> (SkeletalMesh:Skeletal Mesh Object Reference))
   SetMorphTarget!(execute=[@n10.Is Not Valid, @n11.Completed], self:Anim Instance Object Reference, MorphTargetName:Name=shoeScale, Value:float=0) -> (then)
   > ClearAllSkinWeightProfiles(self:Skinned Mesh Component Object Reference=Get FaceAcc_Mesh() -> (FaceAcc_Mesh:Skeletal Mesh Component Budgeted Object Reference))
   SetSkinWeightProfile!(
    execute=@n10.Is Valid
    self:Skinned Mesh Component Object Reference=Get FaceAcc_Mesh()
    InProfileName:Name=Alternative
    InLayer:ESkinWeightProfileLayer Enum=Primary
   ) -> (then, ReturnValue:bool)
   > SetSkinWeightProfile(self=Get FaceAcc_Mesh(), InProfileName=shoeWeightA, InLayer=Secondary)
   > SetMorphTarget(MorphTargetName=shoeScale, Value=1)
   n11 = Do Once(execute=@n12.then_0, Reset=@n9.then_1)
   n12 = Sequence(execute=@n8.else)
   n13 = Event OnCharacterCustomizationLoaded()
   > Set FaceAcc_Mesh(FaceAcc_Mesh:Skeletal Mesh Component Budgeted Object Reference=Get Array Item with EFort Custom Part Type(Array=@n13.CustomMeshParts, PartType=Face)) -> (Output_Get:Skeletal Mesh Component Budgeted Object Reference)
   > Set ShoeMesh(ShoeMesh:Skeletal Mesh Component Budgeted Object Reference=Get Array Item with EFort Custom Part Type(Array=@n13.CustomMeshParts, PartType=ExtraPart)) -> (Output_Get:Skeletal Mesh Component Budgeted Object Reference)
   > Set BodyMesh(BodyMesh=Get Array Item with EFort Custom Part Type(Array=@n13.CustomMeshParts))
   n14 = CheckForBackbling()
   > Set isWearingBackbling(isWearingBackbling=OR Boolean(A=@n14, B=@n14.bIsBackblingLong))
   MapRangeClamped(Value=0, InRangeA=95, InRangeB=110)
  ]}
]
variables [
 BodyMesh {type=SkeletalMeshComponent}
 Character Mesh {type=SkeletalMeshComponent}
 FaceAcc_Mesh {type=SkeletalMeshComponentBudgeted}
 FeatherFix {type=number}
 Gravity Override {type=Vector}
 Gravity Override Params {type=GravityOverrideParamsStruct}
 Is Parachute Opened {type=number}
 isWearingBackbling {type=boolean}
 ShoeMesh {type=SkeletalMeshComponentBudgeted}
 upperarm r rot yaw {type=number}
]
